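# Cheap translucent unlit material for weapon impact flash particles, usable on Niagara and Cascade sprites.
# Particle color starts boosted and eases to its plain value over particle lifetime, while mobile uses plain color through a shading path switch.
# That color passes through two chained MF_FX_EmissiveAdjustments calls.
# Particle alpha multiplies the 2x2 flash texture's green channel, whose lifetime lerp has identical inputs and changes nothing.
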Material /Game/Weapons/FORT_Generic_Effects/Impacts/M_Impact_Flash_Cheap_01
graph {
 properties {
  bUsedWithNanite=true
  bUsedWithNiagaraSprites=true
  bUsedWithParticleSprites=true
  blendMode=BLEND_Translucent
  pixelDepthOffsetMode=PDOM_Legacy
  shadingModel=MSM_Unlit
  usedShadingModels=Unlit
 }
 nodes [
  n1 = ParticleColor()
  n2 = ParticleRelativeTime()
  n3 = LinearInterpolate(A=Multiply(B=@n1.RGB) {constA=4}, B=@n1.RGB, Alpha=@n2) {constA=6}
  MaterialFunctionCall(Emissive=MaterialFunctionCall(Emissive=ShadingPathSwitch(Default=@n3, Mobile=@n1.RGB)) {materialFunction=&MF_FX_EmissiveAdjustments}.Result) {materialFunction=&MF_FX_EmissiveAdjustments}
  n4 = TextureSample() {texture=/Game/Effects/Fort_Effects/Textures/Explo/T_Impact_Flash_2x2_01}
  Multiply(A=@n1.A, B=LinearInterpolate(A=@n4.G, B=@n4.G, Alpha=@n2))
 ]
}
