Material /Game/Weapons/FORT_Generic_Effects/Impacts/M_Impact_Flash_Detail
graph {
 properties {
  bUsedWithNanite=true
  bUsedWithNiagaraSprites=true
  bUsedWithParticleSprites=true
  blendMode=BLEND_Translucent
  pixelDepthOffsetMode=PDOM_Legacy
  shadingModel=MSM_Unlit
  usedShadingModels=Unlit
 }
 nodes [
  n1 = ParticleColor()
  n2 = TextureSampleParameter2D() {parameterName=ImpactTexture, texture=/Game/Effects/Fort_Effects/Textures/Explo/T_Impact_Flash_2x2_01}
  n3 = Add(A=LinearInterpolate(Alpha=@n1.A) {constA=-0.6, constB=0.6}, B=@n2.B)
  n4 = Multiply(A=@n1.A, B=Clamp(None=Multiply(A=@n3, B=@n2.G)))
  LinearInterpolate(
   A=@n4
   B=DepthFade(Opacity=@n4, FadeDistance=ScalarParameter() {defaultValue=25, parameterName=DepthFade}) {fadeDistanceDefault=25}
   Alpha=ScalarParameter() {parameterName=UseDepthFade}
  )
  n5 = Multiply(A=ScalarParameter() {defaultValue=4, parameterName="Particle color start  Multiply"}, B=@n1.RGB) {constA=4}
  n6 = ParticleRelativeTime()
  MaterialFunctionCall(Emissive=ShadingPathSwitch(Default=LinearInterpolate(A=@n5, B=@n1.RGB, Alpha=@n6) {constA=6}, Mobile=@n1.RGB)) {materialFunction=&MF_FX_EmissiveAdjustments}
  n7 = ParticlePositionWS()
  n8 = Subtract(A=@n7, B=WorldPosition().XYZ)
  n9 = ScalarParameter() {defaultValue=5, parameterName=MaxSize}
  n10 = CameraPositionWS()
  n11 = Divide(
   A=Add(A=Distance(A=@n7, B=@n10), B=ScalarParameter() {defaultValue=-1250, parameterName=ScaleDistanceOffset})
   B=ScalarParameter() {defaultValue=20000, parameterName=Scale Distance}
  )
  ShadingPathSwitch(Default=Multiply(A=Multiply(A=@n8, B=@n9), B=Clamp(None=@n11)), Mobile=Constant())
 ]
}
