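# Translucent unlit particle material for wood impact debris, usable on particle and Niagara sprites.
# A SubUV function picks a frame from a 2x2 debris atlas, using the Random dynamic parameter scaled by a frame-count multiplier.
# The texture's green channel tints particle color, which feeds an emissive-adjustment function through a "Fake Lighting" multiplier.
# The same channel, boosted and clamped, multiplies particle alpha to form an opacity mask.
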Material /Game/Weapons/FORT_Generic_Effects/Impacts/M_Impact_Wood_GPU_Debris_01
graph {
 properties {
  bUsedWithNanite=true
  bUsedWithNiagaraSprites=true
  bUsedWithParticleSprites=true
  blendMode=BLEND_Translucent
  pixelDepthOffsetMode=PDOM_Legacy
  shadingModel=MSM_Unlit
  usedShadingModels=Unlit
 }
 nodes [
  n1 = ScalarParameter() {defaultValue=2, parameterName=Grid}
  n2 = DynamicParameter() {paramNames [Random,Param2,Param3,Param4]}
  n3 = ScalarParameter() {defaultValue=4, parameterName=Mult_NrFrame}
  n4 = TextureSample(UVs=MaterialFunctionCall!(
    U_Images=@n1
    V_Images=@n1
    SubUV Texture RGB
    Time Speed=Constant() {r=1}
    Manual Time Control=StaticBool() {value=true}
    Time=Multiply(A=@n2.Random, B=@n3)
    UVs
   ) {materialFunction=/Game/Effects/Fort_Effects/Materials/_Functions/MF_SubUV}.UVs) {texture=/Game/Effects/Fort_Effects/Textures/Wood/T_Impact_Wood_Debris_2x2_01}
  n5 = Multiply(A=ParticleColor().RGB, B=@n4.G)
  MaterialFunctionCall(Emissive=Multiply(A=@n5, B=ScalarParameter() {defaultValue=1, parameterName=Fake Lighting})) {materialFunction=&MF_FX_EmissiveAdjustments}
  Multiply(A=@n5, B=ScalarParameter() {parameterName=Particle Color As Base Color})
  Multiply(A=ParticleColor().A, B=Clamp(None=Multiply(B=@n4.G) {constA=4}))
  Time() {bOverride_Period=true, desc=For testing, period=1}
 ]
}
