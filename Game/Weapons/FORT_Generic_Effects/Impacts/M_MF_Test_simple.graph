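# Translucent, unlit, two-sided test material for impact effect particles, usable on sprites and meshes.
# Channels of a packed impact sheet blend particle color, an inner color and a dark outline color, then feed MF_FX_EmissiveAdjustments.
# A camera-facing falloff fades edge-on surfaces by scaling particle and texture alpha, and is bypassed at Low quality.
# The pixel-depth blend has no effect because both lerp inputs are the same channel.
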Material /Game/Weapons/FORT_Generic_Effects/Impacts/M_MF_Test_simple
graph {
 properties {
  bEnableResponsiveAA=true
  bUsedWithMeshParticles=true
  bUsedWithNanite=true
  bUsedWithNiagaraMeshParticles=true
  bUsedWithParticleSprites=true
  blendMode=BLEND_Translucent
  pixelDepthOffsetMode=PDOM_Legacy
  shadingModel=MSM_Unlit
  twoSided=true
  usedShadingModels=Unlit
 }
 nodes [
  n1 = Constant() {r=1}
  n2 = TextureSample() {samplerType=SAMPLERTYPE_LinearColor, texture=&/T_Impact_Sheet_Packed}
  n3 = Add(A=PixelDepth().R) {constB=-900}
  n4 = LinearInterpolate(A=@n2.G, B=@n2.G, Alpha=Clamp(None=Divide(A=@n3) {constB=100}))
  n5 = DotProduct(A=CameraVectorWS(), B=VertexNormalWS())
  n6 = Clamp(None=Power(Base=Multiply(A=@n5) {constB=2.4}) {constExponent=10})
  n7 = LinearInterpolate(A=@n1, B=@n4, Alpha=@n6)
  n8 = ParticleColor()
  n9 = Multiply(A=Multiply(A=QualitySwitch(Default=@n7, Low=@n1, High=@n7, Medium=@n7)) {constB=1.5, desc=2}, B=@n8.RGB)
  n10 = OneMinus(None=@n2.B)
  n11 = Multiply(A=VectorParameter() {parameterName=Color Inner, defaultValue=<rgba 1 0.3017 0.0007 1>}.RGB, B=@n10)
  MaterialFunctionCall(Emissive=LinearInterpolate(
    A=VectorParameter() {parameterName=Color Outline, defaultValue=<rgba 0.0059 0.0191 0.0442 1>}.RGB
    B=LinearInterpolate(A=@n9, B=@n11, Alpha=@n10) {desc=Learp in Color ramp on edges}
    Alpha=@n4
   ) {desc=blends back in black outline}) {materialFunction=&MF_FX_EmissiveAdjustments}
  Multiply(A=Multiply(A=@n8.A, B=@n2.A), B=QualitySwitch(Default=@n6, Low=Constant() {r=1}, High=@n6, Medium=@n6))
 ]
}
