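# Translucent unlit sprite material for Niagara and Cascade particles, drawing a ring bounce effect, apparently for Space Pistol impacts.
# A noise texture slightly distorts the UVs that sample a channel-packed ring texture.
# Opacity combines red times green raised to a dynamic Power with blue scaled by dynamic Sphere Middle, all multiplied by particle alpha.
# Particle color alone feeds emissive through a shared FX adjustment function, and responsive AA is enabled.
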Material /Game/Weapons/FORT_Generic_Effects/Impacts/M_RingBounce_SpacePistol
graph {
 properties {
  bEnableResponsiveAA=true
  bUsedWithNiagaraSprites=true
  bUsedWithParticleSprites=true
  blendMode=BLEND_Translucent
  opacityMaskClipValue=0.5
  pixelDepthOffsetMode=PDOM_Legacy
  shadingModel=MSM_Unlit
  usedShadingModels=Unlit
 }
 nodes [
  n1 = ParticleColor()
  MaterialFunctionCall(Emissive=@n1.RGB) {materialFunction=&MF_FX_EmissiveAdjustments}
  n2 = TextureSample() {texture=&T_NoisePack_01}
  n3 = Subtract(A=Multiply(A=MaterialFunctionCall(X=@n2.R, Y=@n2.R) {materialFunction=&MakeFloat2}.Result) {constB=2})
  n4 = TextureSample(UVs=Add(A=TextureCoordinate(), B=Multiply(A=@n3) {constB=0.012})) {texture=&/T_RingBounce_02}
  n5 = DynamicParameter() {defaultValue=<rgba 0 0 0 1>, paramNames [Power, Sphere Middle, Param3, Param4]}
  n6 = Power(Base=@n4.G, Exp=@n5.Power)
  Multiply(A=@n1.A, B=Add(A=Multiply(A=@n4.R, B=@n6), B=Multiply(A=@n4.B, B=@n5.Sphere Middle)))
 ]
}
