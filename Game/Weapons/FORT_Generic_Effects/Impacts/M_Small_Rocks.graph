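# Masked, Default Lit particle material for small rock sprites in weapon impact effects, usable with Cascade and Niagara sprites.
# A SubUV rock texture supplies the color for ConvertFromDiffSpec and, multiplied by particle color, the input to MF_FX_EmissiveAdjustments.
# Particle alpha is multiplied by texture alpha.
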Material /Game/Weapons/FORT_Generic_Effects/Impacts/M_Small_Rocks
graph {
 properties {
  bUsedWithNiagaraSprites=true
  bUsedWithParticleSprites=true
  blendMode=BLEND_Masked
  pixelDepthOffsetMode=PDOM_Legacy
  usedShadingModels=Default Lit
 }
 nodes [
  n1 = ParticleColor()
  n2 = ParticleSubUV!(UVs, Tex, Apply View MipBias) {texture=/Game/Effects/Fort_Effects/Textures/Rocks/T_Small_Rock_SubUV_1}
  MaterialFunctionCall(Emissive=Multiply(A=@n1.RGB, B=@n2.RGB)) {materialFunction=&MF_FX_EmissiveAdjustments}
  MaterialFunctionCall(DiffuseColor=@n2.RGB) {materialFunction=&ConvertFromDiffSpec}
  Multiply(A=@n1.A, B=@n2.A)
 ]
}
