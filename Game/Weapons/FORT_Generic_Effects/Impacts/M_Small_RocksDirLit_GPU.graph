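# Translucent, directionally lit sprite material for small rock chunks in weapon impact effects, usable by Cascade and Niagara sprites.
# A random sub-UV function picks a cell of the blurred 2x2 chunk atlas per particle, while the mobile shading path uses fixed half-tiled coordinates.
# The atlas red channel is tinted by BaseColor and particle color, and green is multiplied by particle alpha.
# A matching normal map shares those UVs.
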Material /Game/Weapons/FORT_Generic_Effects/Impacts/M_Small_RocksDirLit_GPU
graph {
 properties {
  bUsedWithNiagaraSprites=true
  bUsedWithParticleSprites=true
  blendMode=BLEND_Translucent
  pixelDepthOffsetMode=PDOM_Legacy
  translucencyLightingMode=TLM_VolumetricPerVertexDirectional
  usedShadingModels=Default Lit
 }
 nodes [
  n1 = ParticleColor()
  n2 = ScalarParameter() {defaultValue=2, parameterName=Rows_Columns}
  n3 = ShadingPathSwitch(
   Default=MaterialFunctionCall!(UseTexCoord 1, Rows=@n2, Columns=@n2, Particle Random Value=ParticleRandom()) {materialFunction=/Game/Effects/Fort_Effects/Materials/_Functions/MF_RandomSubUV_GPU_Sprites}.Result
   Mobile=TextureCoordinate() {uTiling=0.5, vTiling=0.5}
  ) {desc=Migrated from FeatureLevelSwitch}
  n4 = TextureSample(UVs=@n3) {texture=/Game/Effects/Fort_Effects/Textures/Rocks/T_ChunksBlur_2x2_D}
  Multiply(A=@n1.A, B=@n4.G)
  Multiply(A=Multiply(A=VectorParameter() {parameterName=BaseColor, defaultValue=<rgba 0.5 0.5 0.5 1>}.RGB, B=@n4.R), B=@n1.RGB)
  TextureSample(UVs=@n3) {samplerType=SAMPLERTYPE_Normal, texture=/Game/Effects/Fort_Effects/Textures/Rocks/T_ChunksBlur_2x2_N}
 ]
}
