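# Translucent lit sprite material for small-rock impact particles, built for Niagara sprite renderers.
# A SubUV rock flipbook tinted by particle color supplies the color, and a copy scaled by an Emissive dynamic parameter passes through MF_FX_EmissiveAdjustments.
# Alpha multiplies particle alpha, exponent-adjusted texture alpha and a near-camera fade with a tunable distance to soften rocks close to the view.
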Material /Game/Weapons/FORT_Generic_Effects/Impacts/M_Small_Rocks_Transparent
graph {
 properties {bUsedWithNiagaraSprites=true, blendMode=BLEND_Translucent, pixelDepthOffsetMode=PDOM_Legacy, usedShadingModels=Default Lit}
 nodes [
  DynamicParameter() {paramNames [Emissive,Test,Testing,Param4]}
  n1 = ParticleColor()
  n2 = ParticleSubUV!(UVs, Tex, Apply View MipBias) {texture=/Game/Effects/Fort_Effects/Textures/Rocks/T_Small_Rock_SubUV_1}
  MaterialFunctionCall(Emissive=Multiply(A=Multiply(A=@n1.RGB, B=@n2.RGB), B=DynamicParameter() {paramNames [Emissive,Test,Testing,Param4]}.Emissive)) {materialFunction=&MF_FX_EmissiveAdjustments}
  n3 = ScalarParameter() {defaultValue=1, parameterName=Exp}
  Multiply(
   A=Multiply(A=@n1.A, B=Power(Base=@n2.A, Exp=@n3))
   B=MaterialFunctionCall(Distance=ScalarParameter() {defaultValue=500, parameterName="Distance (S)"}) {materialFunction=&Near_Camera_Fade}.Result
  )
  Multiply(A=@n2.RGB, B=@n1.RGB)
  ScalarParameter()
 ]
}
